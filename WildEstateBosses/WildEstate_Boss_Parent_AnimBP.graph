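# Parent AnimBlueprint for Wild Estate bosses, supplying Item Override layers on the Fortnite player skeleton with placeholder non-combat animations.
# Each update caches the pawn, sets Alerted and EnterCombat from its Event.Animation gameplay tags, and forces IsFlying true.
# ItemFinalPoseOverride passes the incoming pose through in combat, otherwise playing the NonCombatState machine.
# That machine runs from spawn into an idle loop, with enter, loop and exit states driven by Alerted.
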AnimBlueprint /WildEstateBosses/WildEstate_Boss_Parent_AnimBP
target_skeleton=&Fortnite_M_Avg_Player_Skeleton
graphs [
 AnimGraph {nodes [Output Pose()]}
 EventGraph {nodes [
   Event BlueprintUpdateAnimation()
   > n1 = Sequence!(execute) -> (then_0, then_1, then_2:Exec, then_3:Exec)
   > n2 = Cast To FortPlayerPawn(Object=TryGetPawnOwner())
   > Set As Fort Player Pawn(As Fort Player Pawn:Fort Player Pawn Object Reference=@n2) -> (Output_Get:Fort Player Pawn Object Reference)
   Set IsFlying(execute=@n1.then_3, IsFlying:bool=true) -> (Output_Get:bool)
   n3 = Get As Fort Player Pawn() -> (As Fort Player Pawn:Fort Player Pawn Object Reference)
   Set Alerted(
    execute=@n1.then_1
    Alerted:bool=HasMatchingGameplayTag!(self:Gameplay Tag Asset Interface Interface=@n3, TagToCheck:Gameplay Tag Structure=(TagName="Event.Animation.Alerted")) -> (ReturnValue:bool)
   ) -> (Output_Get:bool)
   > Set EnterCombat(EnterCombat:bool=HasMatchingGameplayTag(self=@n3, TagToCheck=(TagName="Event.Animation.Combat"))) -> (Output_Get:bool)
  ]}
 ItemAimOffset {nodes [
   Input Pose!() -> (Pose, AimOffsetAlpha:double, Yaw:double, Pitch:double) {name=InPose_AimOffset}
   Input Pose!() -> (Pose) {name=InPose_PreAimOffsets}
   Output Pose() {layerGroup=Item Override}
  ]}
 ItemFallAdditive {nodes [
   Input Pose() {name=InPoseItemFallAdditive}
   Output Pose() {layerGroup=Item Override}
  ]}
 ItemFallLandAdditive {nodes [
   Input Pose() {name=InPoseItemFallLandAdditive}
   Output Pose() {layerGroup=Item Override}
  ]}
 ItemFallLevel2Override {nodes [
   Input Pose() {name=InPose_FallLevel2}
   Output Pose() {layerGroup=Item Override}
  ]}
 ItemFinalPoseOverride {nodes [
   AnimGraphNode_StateMachine_0 = NonCombatState() {is=title AnimGraphNode_StateMachine}
   Output Pose(Result=Blend Poses by bool(
     BlendPose_0=Use cached pose 'FinalPoseCache'()
     BlendPose_1=@AnimGraphNode_StateMachine_0
     bActiveValue=Get EnterCombat() -> (EnterCombat:bool)
    )) {layerGroup=Item Override}
   Save cached pose 'FinalPoseCache'(Pose=Input Pose() {name=InFinalPose}) {cacheName=FinalPoseCache}
  ]}
 NonCombatState {
  entry=SpawnInto
  path=ItemFinalPoseOverride.AnimGraphNode_StateMachine_0.NonCombatState
  ids=id
  nodes [
   AnimStateNode_0 = SpawnInto() {is=title AnimStateNode}
   AnimStateNode_1 = AltertedLoop() {is=title AnimStateNode}
   AnimStateNode_2 = NonCombatLoop() {is=title AnimStateNode}
   AnimStateNode_3 = ExitAlertedToNonCombat() {is=title AnimStateNode}
   AnimStateNode_4 = EnterAlterted() {is=title AnimStateNode}
   AnimStateTransitionNode_5 = AnimStateTransitionNode() {bAllowInertializationForSelfTransitions=true, bAutomaticRuleBasedOnSequencePlayerInState=true}
   AnimStateTransitionNode_6 = AnimStateTransitionNode() {bAllowInertializationForSelfTransitions=true}
   AnimStateTransitionNode_7 = AnimStateTransitionNode() {bAllowInertializationForSelfTransitions=true, bAutomaticRuleBasedOnSequencePlayerInState=true}
   AnimStateTransitionNode_8 = AnimStateTransitionNode() {bAllowInertializationForSelfTransitions=true}
   AnimStateTransitionNode_9 = AnimStateTransitionNode() {bAllowInertializationForSelfTransitions=true, bAutomaticRuleBasedOnSequencePlayerInState=true}
  ]
  links [
   {
    dst=ExitAlertedToNonCombat
    rule=ItemFinalPoseOverride.AnimGraphNode_StateMachine_0.NonCombatState.AnimStateTransitionNode_8.Transition
    src=AltertedLoop
    via=AnimStateTransitionNode_8
   }
   {
    dst=AltertedLoop
    rule=ItemFinalPoseOverride.AnimGraphNode_StateMachine_0.NonCombatState.AnimStateTransitionNode_7.Transition
    src=EnterAlterted
    via=AnimStateTransitionNode_7
   }
   {
    dst=NonCombatLoop
    rule=ItemFinalPoseOverride.AnimGraphNode_StateMachine_0.NonCombatState.AnimStateTransitionNode_9.Transition
    src=ExitAlertedToNonCombat
    via=AnimStateTransitionNode_9
   }
   {
    dst=EnterAlterted
    rule=ItemFinalPoseOverride.AnimGraphNode_StateMachine_0.NonCombatState.AnimStateTransitionNode_6.Transition
    src=NonCombatLoop
    via=AnimStateTransitionNode_6
   }
   {
    dst=NonCombatLoop
    rule=ItemFinalPoseOverride.AnimGraphNode_StateMachine_0.NonCombatState.AnimStateTransitionNode_5.Transition
    src=SpawnInto
    via=AnimStateTransitionNode_5
   }
  ]
 }
 SpawnInto {
  path=ItemFinalPoseOverride.AnimGraphNode_StateMachine_0.NonCombatState.AnimStateNode_0.SpawnInto
  nodes [Output Animation Pose(Result=Sequence Player 'WildEstateBoss_Placeholder'() {bLoopAnimation=false, sequence=/WildEstateBosses/Assets/Animation/WildEstateBoss_Placeholder})]
 }
 AltertedLoop {
  path=ItemFinalPoseOverride.AnimGraphNode_StateMachine_0.NonCombatState.AnimStateNode_1.AltertedLoop
  nodes [Output Animation Pose(Result=Sequence Player 'WildEstateBoss_Placeholder'() {sequence=&WildEstateBoss_Placeholder})]
 }
 NonCombatLoop {
  path=ItemFinalPoseOverride.AnimGraphNode_StateMachine_0.NonCombatState.AnimStateNode_2.NonCombatLoop
  nodes [Output Animation Pose(Result=Sequence Player 'WildEstateBoss_Placeholder'() {sequence=&WildEstateBoss_Placeholder})]
 }
 ExitAlertedToNonCombat {
  path=ItemFinalPoseOverride.AnimGraphNode_StateMachine_0.NonCombatState.AnimStateNode_3.ExitAlertedToNonCombat
  nodes [Output Animation Pose(Result=Sequence Player 'WildEstateBoss_Placeholder'() {bLoopAnimation=false, sequence=&WildEstateBoss_Placeholder})]
 }
 EnterAlterted {
  path=ItemFinalPoseOverride.AnimGraphNode_StateMachine_0.NonCombatState.AnimStateNode_4.EnterAlterted
  nodes [Output Animation Pose(Result=Sequence Player 'WildEstateBoss_Placeholder'() {bLoopAnimation=false, sequence=&WildEstateBoss_Placeholder})]
 }
 Transition {path=ItemFinalPoseOverride.AnimGraphNode_StateMachine_0.NonCombatState.AnimStateTransitionNode_5.Transition, nodes [Result()]}
 Transition {
  path=ItemFinalPoseOverride.AnimGraphNode_StateMachine_0.NonCombatState.AnimStateTransitionNode_6.Transition
  nodes [
   Result(bCanEnterTransition=Get Alerted() -> (Alerted:bool))
   Get EnterCombat()
  ]
 }
 Transition {path=ItemFinalPoseOverride.AnimGraphNode_StateMachine_0.NonCombatState.AnimStateTransitionNode_7.Transition, nodes [Result()]}
 Transition {
  path=ItemFinalPoseOverride.AnimGraphNode_StateMachine_0.NonCombatState.AnimStateTransitionNode_8.Transition
  nodes [
   Result(bCanEnterTransition=NOT Boolean(A=Get Alerted()))
   Get EnterCombat()
  ]
 }
 Transition {path=ItemFinalPoseOverride.AnimGraphNode_StateMachine_0.NonCombatState.AnimStateTransitionNode_9.Transition, nodes [Result()]}
 ItemFullBodyAimOffsets {nodes [
   Input Pose!() -> (Pose, FullBody_AimOffsetAlpha:double, Aim_Yaw:double, Aim_Pitch:double) {name=InPoseFullBodyAimOffsets}
   Input Pose!() -> (Pose) {name=InPosePreFullBodyAimOffsets}
   Output Pose() {layerGroup=Item Override}
  ]}
 ItemFullBodyOverride {nodes [
   Input Pose() {name=InPoseFullBody}
   Output Pose() {layerGroup=Item Override}
  ]}
 ItemIdleOverride {nodes [
   Input Pose() {name=InPose_Idle}
   Output Pose() {layerGroup=Item Override}
  ]}
 ItemJetPackJumpAdditive {nodes [Output Pose() {layerGroup=Item Override}]}
 ItemJetPackStartAdditive {nodes [Output Pose() {layerGroup=Item Override}]}
 ItemJumpLoopAdditive {nodes [
   Input Pose() {name=InPoseItemJumpLoopAdditive}
   Output Pose() {layerGroup=Item Override}
  ]}
 ItemJumpOverride {nodes [
   Input Pose() {name=InPoseJump}
   Output Pose() {layerGroup=Item Override}
  ]}
 ItemJumpUpAdditive {nodes [
   Input Pose() {name=InPoseItemJumpUpAdditive}
   Output Pose() {layerGroup=Item Override}
  ]}
 ItemPreAimOffsets {nodes [
   Input Pose() {name=InPosePreAimOffsets}
   Output Pose() {layerGroup=Item Override}
  ]}
 ItemPreIK {nodes [
   Input Pose() {name=InPose_PreIK}
   Output Pose() {layerGroup=Item Override}
  ]}
 ItemRightHandIK {nodes [
   Input Pose() {name=InPosePreRightHandIK}
   Input Pose() {name=InPosePostRightHandIK}
   Output Pose() {layerGroup=Item Override}
  ]}
 ItemSkeletalControl {nodes [
   Input Pose()
   Output Pose() {layerGroup=Item Override}
  ]}
 ItemSwimJumpFallLoop {nodes [Output Pose() {layerGroup=Item Override}]}
 ItemSwimJumpSurfaceEnd {nodes [Output Pose() {layerGroup=Item Override}]}
 ItemSwimJumpSurfaceLoop {nodes [Output Pose() {layerGroup=Item Override}]}
 ItemUpperBody {nodes [
   Input Pose!() -> (Pose, InputParam:Fort Anim Input Adjusted Aim Structure) {name=InPoseUpperBody}
   Output Pose() {layerGroup=Item Override}
  ]}
 ItemZipLineStartAdditive {nodes [Output Pose() {layerGroup=Item Override}]}
]
variables [
 Alerted {type=boolean}
 As Fort Player Pawn {type=FortPlayerPawn}
 EnterCombat {type=boolean}
 IsFlying {type=boolean}
]
